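# Stereo MetaSound source for the PlacidMelon mask's charge fire sound, classed as a ranged weapon sound.
# On play it picks randomly from four fire waves via a shared random-sound graph.
# A mixer scales the result by the character's affiliation, with a custom Enemy level, before a master gain feeds the stereo output.
# It uses PlacidMelon player attenuation and a per-owner concurrency limit of one voice, stopping the oldest.
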MetaSoundSource /PlacidMelon/Gameplay/PlacidMelon/Audio/MSS_Mask_PlacidMelon_Charge_Fire
inputs [
 Character.Affiliation {type=Enum:FortAudioAffiliation}
 Character.Health {type=Float}
 Character.UniqueID {type=Int32}
 MasterGain {type=Float}
 UE.Source.OnPlay {type=Trigger}
 Weapons.Fire {type=WaveAsset:Array}
]
interfaces [Fort.Character:1.0, UE.OutputFormat.Stereo:1.0, UE.Source:1.1]
outputs [
 "UE.OutputFormat.Stereo.Audio:0" {type=Audio}
 "UE.OutputFormat.Stereo.Audio:1" {type=Audio}
]
pages [Default {
  input_defaults {
   Character.Affiliation=0
   Character.Health=0
   Character.UniqueID=0
   MasterGain=1.5
   UE.Source.OnPlay=null
   Weapons.Fire [
    &/PlacidMelon_Charge_Fire_01
    &/PlacidMelon_Charge_Fire_02
    &/PlacidMelon_Charge_Fire_03
    &/PlacidMelon_Charge_Fire_04
   ]
  }
  nodes [
   n1 = &MSS_PlayRandomSound_Stereo(UE.Source.OnPlay=^UE.Source.OnPlay, Sounds=^Weapons.Fire)
   AudioMixerAudioMixerStereo3_0 = "AudioMixer.Audio Mixer (Stereo, 3)"(
    Gain 0=&MSP_AffiliationMixFloat(Affiliation=^Character.Affiliation) {also {defaults {Enemy=1.5}}}.Out
    In 0 L=@n1."UE.OutputFormat.Stereo.Audio:0"
    In 0 R=@n1."UE.OutputFormat.Stereo.Audio:1"
   ) {also {defaults {Gain 1=0, Gain 2=0}}}
   MSP_StereoGain_0 = &MSP_StereoGain("Gain (Linear)"=^MasterGain, Left In=@AudioMixerAudioMixerStereo3_0.Out L, Right In=@AudioMixerAudioMixerStereo3_0.Out R)
  ]
  links [
   {dst="@UE.OutputFormat.Stereo.Audio:0", src=MSP_StereoGain_0.Left Out}
   {dst="@UE.OutputFormat.Stereo.Audio:1", src=MSP_StereoGain_0.Right Out}
  ]
 }]
properties {
 assetImportData=&self:AssetImportData
 attenuationSettings=&/SA_Mask_PlacidMelon_Player
 bLooping=true
 duration=10000
 numChannels=2
 outputFormat=Stereo
 sampleRate=48000
 soundClassObject=&Fort_WeaponRanged
 concurrencyOverrides {bLimitToOwner=true, maxCount=1, resolutionRule=StopOldest}
}
