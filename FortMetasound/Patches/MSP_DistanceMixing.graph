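# MetaSound patch that turns a distance value into three curve-shaped mix values for distance-based audio mixing.
# A MapRange node normalizes Distance against MaxFalloffDistance, with a small non-zero output floor.
# That shared value drives three WaveTableBankEvaluate nodes, each reading a different entry of CurveBank as a falloff curve to output MappedRangeOne, MappedRangeTwo and MappedRangeThree.
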MetaSoundPatch /FortMetasound/Patches/MSP_DistanceMixing
inputs [
 CurveBank {type=WaveTableBankAsset}
 Distance {type=Float}
 MaxFalloffDistance {type=Float}
]
outputs [
 MappedRangeOne {type=Float}
 MappedRangeThree {type=Float}
 MappedRangeTwo {type=Float}
]
pages [Default {
  input_defaults {CurveBank=null, Distance=null, MaxFalloffDistance=1}
  nodes [
   InitVariable.Float()
   InitVariable.Float()
   InitVariable.Float()
   n1 = MapRange.MapRange:Float(In=^Distance, In Range B=^MaxFalloffDistance) {also {defaults {Out Range A=0.001}}}
   WaveTableBankEvaluate_0 = UE.WaveTableBankEvaluate(WaveTableBank=^CurveBank, Input=@n1.Out Value) {also {defaults {Index=2}}}
   WaveTableBankEvaluate_1 = UE.WaveTableBankEvaluate(WaveTableBank=^CurveBank, Input=@n1.Out Value)
   WaveTableBankEvaluate_2 = UE.WaveTableBankEvaluate(WaveTableBank=^CurveBank, Input=@n1.Out Value) {also {defaults {Index=1}}}
  ]
  links [
   {dst="@MappedRangeThree", src=WaveTableBankEvaluate_0.Output}
   {dst="@MappedRangeOne", src=WaveTableBankEvaluate_1.Output}
   {dst="@MappedRangeTwo", src=WaveTableBankEvaluate_2.Output}
  ]
 }]
