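# One-shot stereo MetaSound for PlacidMelon projectile impacts, using an explosives sound class.
# Randomized close, distant and far players trigger together, crossfaded by MSP_DistanceMixing from attenuation distance; the Weapons.Fire input is unused.
# The mix is scaled by MasterGain and an affiliation-based gain, and finishes through a trigger accumulator on all three players.
# Concurrency allows one instance per owner, stopping the oldest.
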MetaSoundSource /PlacidMelon/Gameplay/PlacidMelon/Audio/MSS_Mask_PlacidMelon_Proj_Impact
inputs [
 Character.Affiliation {type=Enum:FortAudioAffiliation}
 Character.Health {type=Float}
 Character.UniqueID {type=Int32}
 MasterGain {type=Float}
 UE.Attenuation.Distance {type=Float}
 UE.Source.OnPlay {type=Trigger}
 Weapons.Fire {type=WaveAsset:Array}
 Weapons.FireClose {type=WaveAsset:Array}
 Weapons.FireDistant {type=WaveAsset:Array}
 Weapons.FireFar {type=WaveAsset:Array}
]
interfaces [Fort.Character:1.0, UE.Attenuation:1.0, UE.OutputFormat.Stereo:1.0, UE.Source.OneShot:1.0, UE.Source:1.1]
outputs [
 "UE.OutputFormat.Stereo.Audio:0" {type=Audio}
 "UE.OutputFormat.Stereo.Audio:1" {type=Audio}
 UE.Source.OneShot.OnFinished {type=Trigger}
]
pages [Default {
  input_defaults {
   Character.Affiliation=0
   Character.Health=0
   Character.UniqueID=0
   MasterGain=1
   UE.Attenuation.Distance=0
   UE.Source.OnPlay=null
   Weapons.Fire [
    &/PlacidMelon_Proj_Impact_01
    &/PlacidMelon_Proj_Impact_02
    &/PlacidMelon_Proj_Impact_03
    &/PlacidMelon_Proj_Impact_04
   ]
   Weapons.FireClose [
    &PlacidMelon_Proj_Impact_01
    &PlacidMelon_Proj_Impact_02
    &PlacidMelon_Proj_Impact_03
    &PlacidMelon_Proj_Impact_04
   ]
   Weapons.FireDistant [
    &/PlacidMelon_Proj_Impact_Distant_01
    &/PlacidMelon_Proj_Impact_Distant_02
    &/PlacidMelon_Proj_Impact_Distant_03
    &/PlacidMelon_Proj_Impact_Distant_04
   ]
   Weapons.FireFar [
    &/PlacidMelon_Proj_Impact_Far_01
    &/PlacidMelon_Proj_Impact_Far_02
    &/PlacidMelon_Proj_Impact_Far_03
    &/PlacidMelon_Proj_Impact_Far_04
   ]
  }
  nodes [
   n1 = /FortMetasound/Patches/MSP_DistanceMixing(Distance=^UE.Attenuation.Distance) {also {defaults {CurveBank=&WeaponRanged_DistanceCurves, MaxFalloffDistance=30000}}}
   n2 = /MetaSoundWizard/MetaSounds/One-shots/ArrayRandomized/MSS_Oneshot_Array_Randomized_Stereo(UE.Source.OnPlay=^UE.Source.OnPlay, Sounds=^Weapons.FireDistant) {also {defaults {Random.Pitch.Spread=0.5, Random.Volume.Spread=0.5}}}
   n3 = &MSS_Oneshot_Array_Randomized_Stereo(UE.Source.OnPlay=^UE.Source.OnPlay, Sounds=^Weapons.FireFar) {also {defaults {Random.Pitch.Spread=0.5, Random.Volume.Base=-9, Random.Volume.Spread=0.5}}}
   n4 = &MSS_Oneshot_Array_Randomized_Stereo(UE.Source.OnPlay=^UE.Source.OnPlay, Sounds=^Weapons.FireClose) {also {defaults {Random.Pitch.Spread=0.5, Random.Volume.Spread=0.5}}}
   AudioMixerAudioMixerStereo3_0 = "AudioMixer.Audio Mixer (Stereo, 3)"(
    Gain 0=@n1.MappedRangeOne
    Gain 2=@n1.MappedRangeThree
    Gain 1=@n1.MappedRangeTwo
    In 1 L=@n2."UE.OutputFormat.Stereo.Audio:0"
    In 1 R=@n2."UE.OutputFormat.Stereo.Audio:1"
    In 2 L=@n3."UE.OutputFormat.Stereo.Audio:0"
    In 2 R=@n3."UE.OutputFormat.Stereo.Audio:1"
    In 0 L=@n4."UE.OutputFormat.Stereo.Audio:0"
    In 0 R=@n4."UE.OutputFormat.Stereo.Audio:1"
   )
   MSP_StereoGain_0 = &MSP_StereoGain(
    Left In=@AudioMixerAudioMixerStereo3_0.Out L
    Right In=@AudioMixerAudioMixerStereo3_0.Out R
    "Gain (Linear)"=UE.Multiply:Float(AdditionalOperands=^MasterGain, PrimaryOperand=&MSP_AffiliationMixFloat(Affiliation=^Character.Affiliation).Out).Out
   )
   TriggerAccumulatorTriggerAccumulate3_0 = "TriggerAccumulator.Trigger Accumulate (3)"(In 1=@n2.UE.Source.OneShot.OnFinished, In 2=@n3.UE.Source.OneShot.OnFinished, In 0=@n4.UE.Source.OneShot.OnFinished)
  ]
  links [
   {dst="@UE.OutputFormat.Stereo.Audio:0", src=MSP_StereoGain_0.Left Out}
   {dst="@UE.OutputFormat.Stereo.Audio:1", src=MSP_StereoGain_0.Right Out}
   {dst="@UE.Source.OneShot.OnFinished", src=TriggerAccumulatorTriggerAccumulate3_0.Out}
  ]
 }]
properties {
 assetImportData=&self:AssetImportData
 attenuationSettings=&/SA_Mask_PlacidMelon_Impact
 numChannels=2
 outputFormat=Stereo
 sampleRate=48000
 soundClassObject=/Game/Sounds/FortSoundClassesAndModes/Fort_Explosives_Large_PSM
 volume=0.25
 concurrencyOverrides {bLimitToOwner=true, maxCount=1, resolutionRule=StopOldest}
}
