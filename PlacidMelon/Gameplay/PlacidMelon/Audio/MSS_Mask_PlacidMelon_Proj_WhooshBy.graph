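# One-shot stereo MetaSound source for the whiz-by sound of PlacidMelon mask projectiles, in the ranged weapon projectile effects sound class.
# On play, a random-sound subgraph picks from three whiz-by waves, a stereo gain applies MasterGain, and the pick's finished trigger ends the source.
# The asset plays quietly at raised pitch with its own attenuation and overrides concurrency to a few instances per owner, stopping the oldest.
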MetaSoundSource /PlacidMelon/Gameplay/PlacidMelon/Audio/MSS_Mask_PlacidMelon_Proj_WhooshBy
inputs [
 Character.Affiliation {type=Enum:FortAudioAffiliation}
 Character.Health {type=Float}
 Character.UniqueID {type=Int32}
 MasterGain {type=Float}
 UE.Source.OnPlay {type=Trigger}
 Weapons.Fire {type=WaveAsset:Array}
]
interfaces [Fort.Character:1.0, UE.OutputFormat.Stereo:1.0, UE.Source.OneShot:1.0, UE.Source:1.1]
outputs [
 "UE.OutputFormat.Stereo.Audio:0" {type=Audio}
 "UE.OutputFormat.Stereo.Audio:1" {type=Audio}
 UE.Source.OneShot.OnFinished {type=Trigger}
]
pages [Default {
  input_defaults {
   Character.Affiliation=0
   Character.Health=0
   Character.UniqueID=0
   MasterGain=1
   UE.Source.OnPlay=null
   Weapons.Fire [
    &/PlacidMelon_Proj_WhizzBy_01
    &/PlacidMelon_Proj_WhizzBy_02
    &/PlacidMelon_Proj_WhizzBy_03
   ]
  }
  nodes [
   MSS_PlayRandomSound_Stereo_0 = &MSS_PlayRandomSound_Stereo(UE.Source.OnPlay=^UE.Source.OnPlay, Sounds=^Weapons.Fire)
   MSP_StereoGain_0 = &MSP_StereoGain(
    "Gain (Linear)"=^MasterGain
    Left In=@MSS_PlayRandomSound_Stereo_0."UE.OutputFormat.Stereo.Audio:0"
    Right In=@MSS_PlayRandomSound_Stereo_0."UE.OutputFormat.Stereo.Audio:1"
   )
  ]
  links [
   {dst="@UE.OutputFormat.Stereo.Audio:0", src=MSP_StereoGain_0.Left Out}
   {dst="@UE.OutputFormat.Stereo.Audio:1", src=MSP_StereoGain_0.Right Out}
   {dst="@UE.Source.OneShot.OnFinished", src=MSS_PlayRandomSound_Stereo_0.UE.Source.OneShot.OnFinished}
  ]
 }]
properties {
 assetImportData=&self:AssetImportData
 attenuationSettings=&/SA_Mask_PlacidMelon_Proj_Whip
 bOverrideConcurrency=true
 maxDistance=1100
 numChannels=2
 outputFormat=Stereo
 pitch=2
 sampleRate=48000
 soundClassObject=&Fort_WeaponRanged_Projectile_Effects
 volume=0.1
 concurrencyOverrides {bLimitToOwner=true, maxCount=3, resolutionRule=StopOldest}
}
