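; Character part modifier for the PlacidMelon weapon's mask, managing its mesh and Niagara effects on the player.
; On post res-in it passes mesh and idle effect to the weapon, attaches that effect, and registers tag waits that hide the equip effect during prop-hide or cloak.
; Customization finish and switches to other weapons stamp game time into custom primitive data for the material.
; Swapping between PlacidMelon weapons reruns res-in setup.
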(event EventOnPostResIn
  (bind _prophidetag (Variables|Default|GetPropHideTag))
  (bind _as_fort_player_pawn_athena (Variables|Default|GetAs%20Fort%20Player%20Pawn%20Athena))
  (|CallParentFunction)
  (Utilities|FlowControl|Sequence
    (:then_0
      (bind _asfort_player_pawn_athena (Utilities|Casting|CastToFortPlayerPawnAthena (CustomCharacter|GetAssociatedPlayerPawn)))
      (Variables|Default|SetAs%20Fort%20Player%20Pawn%20Athena _asfort_player_pawn_athena)
      (bind _asb_weap_placid_melon (Utilities|Casting|CastToB_Weap_PlacidMelon (Variables|/Script/FortniteGame.FortPawn|GetCurrentWeapon _asfort_player_pawn_athena)))
      (Class|BWeapPlacidMelon|SetMask _asb_weap_placid_melon (CustomCharacter|GetSkeletalMeshforPartonOwner) (Variables|Default|GetNS_PlacidMelon_Mask_Idle)))
    (:then_1
      (CallFunction|MaskNS)
      ;[1 1 1] Hide Effects while in Prop
      (Ability|Tasks|WaitGameplayTagAddToActor _as_fort_player_pawn_athena _prophidetag
        (:then
          (Ability|Tasks|WaitGameplayTagRemoveFromActor _as_fort_player_pawn_athena _prophidetag
            (:then
              ;[1 1 1] Turn invisible if the player goes invisible
              (Ability|Tasks|WaitGameplayTagAddToActor _as_fort_player_pawn_athena (Variables|Default|GetCloakTag)
                (:then
                  (Ability|Tasks|WaitGameplayTagRemoveFromActor _as_fort_player_pawn_athena (Variables|Default|GetCloakTag)
                    (:Removed
                      (Rendering|SetVisibility (Variables|Default|GetMask%20Equip%20NS) true true))))
                (:Added
                  (Rendering|SetVisibility (Variables|Default|GetMask%20Equip%20NS) false true))))
            (:Removed
              (Rendering|SetVisibility (Variables|Default|GetMask%20Equip%20NS) true true))))
        (:Added
          (Rendering|SetVisibility (Variables|Default|GetMask%20Equip%20NS) false true))))))

(event Custom|Mask%20NS
  (bind _ns_placidmelon_mask_idle (Variables|Default|GetNS_PlacidMelon_Mask_Idle))
  ;[1 1 1] This is a work around for FORT-589563. A systemic fix described in FORT-617814 should make the original code work without log spam. Note also this work around required the addition of the Glove Equip NS component.
  (if true
    ;[0.4588 0.6885 0.8125 0.5] FX - Hide Orb if FX spawns when inside a vehicle (like switching seats)
    (Components|Activation|Activate _ns_placidmelon_mask_idle false)
    (Transformation|AttachComponentToComponent _ns_placidmelon_mask_idle (CustomCharacter|GetSkeletalMeshForPartOnPlayerPawn) "main" "KeepRelative" "KeepRelative" "KeepRelative" false)))

(event EventOnWeaponChanged
  (params (NewWeapon :type Fort%20Weapon%20Object%20Reference) (OldWeapon :type Fort%20Weapon%20Object%20Reference))
  (bind _returnvalue (CustomCharacter|GetSkeletalMeshForPartOnPlayerPawn))
  (|CallParentFunction NewWeapon OldWeapon)
  (Utilities|FlowControl|Sequence
    (:then_0
      (Utilities|IsValid NewWeapon
        (:Is%20Valid
          (Utilities|Casting|CastToB_Weap_PlacidMelon NewWeapon
            (:CastFailed
              (Utilities|IsValid _returnvalue
                (:Is%20Valid
                  (Rendering|Material|SetCustomPrimitiveDataFloat _returnvalue 1 (Utilities|Time|GetGameTimeinSeconds)))
                (:Is%20Not%20Valid
                  (Development|PrintString "INVALID"))))))))
    (:then_1
      ; This resets the mask FX if replacing mask of same type
      (if (Utilities|IsValid OldWeapon :!class /Script/Engine.KismetSystemLibrary)
        (if (and (Tags|WeaponDataHasAnyMatchingGameplayTags OldWeapon "(GameplayTags=((TagName=\"Item.Weapon.Ranged.PlacidMelon\")))") (Tags|WeaponDataHasAnyMatchingGameplayTags NewWeapon "(GameplayTags=((TagName=\"Item.Weapon.Ranged.PlacidMelon\")))"))
          (CallFunction|OnPostResIn))))))

(event EventOnCharacterCustomizationFinished
  (params (CharacterPartInfo :type Array%20of%20Applied%20Character%20Part%20Info%20Structures))
  (bind _returnvalue (CustomCharacter|GetSkeletalMeshForPartOnPlayerPawn))
  (|CallParentFunction CharacterPartInfo)
  (Utilities|IsValid _returnvalue
    (:Is%20Valid
      (Rendering|SetVisibility _returnvalue true true)
      (Rendering|Material|SetCustomPrimitiveDataFloat _returnvalue 0 (Utilities|Time|GetGameTimeinSeconds)))))

(fn ConstructionScript
  (|CallParentFunction))
