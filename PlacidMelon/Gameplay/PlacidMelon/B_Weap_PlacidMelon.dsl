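; PlacidMelon weapon Blueprint for a charged beam item with a mask effect, consumed as its ammo runs out.
; Equipping sends a gameplay event to the instigator; unequipping detaches the mask Niagara component and removes effects granting the equipped tag.
; Charge start spends one ammo unless infinite, while beam and ability-end functions fire delegates.
; Authority-side ammo tracking and local unequip trigger item removal, delayed to avoid a crash and match animation.
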(event EventBeginPlay
  ;[1 1 1] BeginPlay
  (|CallParentFunction))

(event FortWeapon|EventOnEquippedAndAttached
  ;[1 1 1] Equip
  (|CallParentFunction)
  (Ability|SendGameplayEventtoActor (Game|GetInstigator) "(TagName=\"PlacidMelon.Equipped\")")
  (Default|CallOnEquipped)
  (Utilities|Casting|CastToFortPlayerPawn (Game|GetInstigator)))

(event FortWeapon|EventOnUnEquippedAndDetached
  ;[1 1 1] UnEquip
  (|CallParentFunction)
  (Transformation|DetachFromComponent (Variables|Default|GetMask_NS) "KeepRelative" "KeepRelative" "KeepRelative" true)
  (bind _asfort_player_pawn (Utilities|Casting|CastToFortPlayerPawn (Game|GetInstigator)))
  (GameplayEffects|RemoveActiveEffectswithGrantedTags (Ability|GetAbilitySystemComponent _asfort_player_pawn) "(GameplayTags=((TagName=\"PlacidMelon.Equipped\")))"))

(event Custom|SetMask
  (params (Mesh_Mask :type Mesh%20Component%20Object%20Reference) (NewParam :type Niagara%20Particle%20System%20Component%20Object%20Reference))
  (Variables|Default|SetMesh_Mask Mesh_Mask)
  (Variables|Default|SetMask_NS NewParam))

(event Custom|OnConsumeAmmoTracking
  (params (Quantity :type Integer))
  (bind _returnvalue (FortWeapon|GetMagazineAmmoCount))
  (if (ScalableFloat|GetValueasBool (Variables|Default|GetRow_AmmoConsumption_Enabled))
    (if (not (CallFunction|IsAmmoInfinite))
      (Networking|SwitchHasAuthority
        (:Authority
          (Variables|Default|SetLocalBPAmmo _returnvalue)
          (if (<= (Math|Float|SelectFloat (Math|Conversions|ToFloat(Integer) (FortWeapon|GetBulletsPerClip)) (Math|Conversions|ToFloat(Integer) (FortWeapon|GetRemainingAmmo)) (FortWeapon|Ammo|HasInfiniteMagazineUses)) 0)
            (CallFunction|RemoveItemOnNoAmmoRemaining)))
        (:Remote
          (Variables|Default|SetLocalBPAmmo _returnvalue))))))

(event FortWeapon|EventOnLocalUnEquip
  (CallFunction|RemoveItemOnNoAmmoRemaining))

(event Custom|Remove%20Item%20on%20No%20Ammo%20Remaining
  (if (<= (FortWeapon|GetRemainingAmmo) 0)
    (bind _asfort_player_pawn (Utilities|Casting|CastToFortPlayerPawn (Game|GetInstigator)))
    ;[1 1 1] Delay to avoid mantis crash and to sync with animation removal
    (Utilities|FlowControl|Delay 0.7)
    (Inventory|RemoveItemFromPlayerByGuid (Fortnite|GetFortPlayerControllerfromActor _asfort_player_pawn) (FortWeapon|GetInventoryGUID) 1 true)))

(fn AbilityEnd
  (Default|CallAbilityFinished))

(fn BeamFired
  (Default|CallBeamFiredDelegate))

(fn ChargeStart
  (if (not (CallFunction|IsAmmoInfinite))
    (FortWeapon|Ammo|ConsumeAmmo 1)))

(fn DestroySelf
  (bind _returnvalue (Game|GetInstigatorController))
  (Utilities|IsValid (Variables|Actor|GetInstigator)
    (:Is%20Valid
      (Utilities|IsValid _returnvalue
        (:Is%20Valid
          (Inventory|RemoveItemFromPlayerByGuid _returnvalue (FortWeapon|GetInventoryGUID) 1 true))))))

(fn IsAmmoInfinite
  (return (or (FortWeapon|Ammo|HasInfiniteCharges) (FortWeapon|Ammo|InventoryOwnerHasInfiniteUses))))

(fn OnRep_BeamEnd)

(fn ConstructionScript)
