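# Translucent unlit beam effect material for Niagara ribbons and mesh particles, with a scrolling noisy core and soft edges.
# Two panning noise samples, U-scaled by a dynamic parameter, multiply into a V-centered sphere mask to form the core.
# Opacity adds a soft edge, subtracts Dissolve, then applies sine-based UV border fading, particle alpha and depth fade.
# The core mask blends edge and core colors, tinted by particle color through MF_FX_EmissiveAdjustments.
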Material /PlacidMelon/Gameplay/PlacidMelon/FX/Materials/M_PlacidMelon_Beam
graph {
 properties {
  bAutomaticallySetUsageInEditor=false
  bUsedWithNiagaraMeshParticles=true
  bUsedWithNiagaraRibbons=true
  blendMode=BLEND_Translucent
  shadingModel=MSM_Unlit
  usedShadingModels=Unlit
 }
 nodes [
  n1 = ParticleColor()
  n2 = TextureCoordinate() {uTiling=0}
  n3 = Constant2Vector() {g=0.5}
  n4 = ScalarParameter() {defaultValue=0.5, parameterName=Core Radius}
  n5 = AppendVector(A=DynamicParameter() {defaultValue=<rgba 2 0 1 1>, paramNames [UScale,Dissolve,Param3,Param4]}.UScale, B=Constant() {r=1})
  n6 = TextureCoordinate() {uTiling=1.5}
  n7 = Multiply(A=@n5, B=@n6)
  n8 = Multiply(A=@n6, B=@n5)
  n9 = Multiply(
   A=TextureSample(UVs=Panner(Coordinate=@n7) {speedX=-5}) {texture=&T_Smooth_tile}.R
   B=TextureSample(UVs=Panner(Coordinate=@n8) {speedX=-0.15}) {texture=&T_Smooth_tile}.R
  )
  n10 = Multiply(
   A=Power(Base=SphereMask(A=@n2, B=@n3, Radius=@n4) {attenuationRadius=0.5, hardnessPercent=0}) {constExponent=3}
   B=Add(A=Multiply(A=@n9) {constB=4}, B=ScalarParameter() {parameterName=Cancel Noise}) {constB=0}
  )
  n11 = Multiply(
   A=SphereMask(A=TextureCoordinate() {uTiling=0}, B=Constant2Vector() {g=0.5}) {attenuationRadius=0.5, hardnessPercent=0}
   B=ScalarParameter() {defaultValue=0.5, parameterName=Soft Edge Multiply}
  ) {constB=0.5}
  n12 = Subtract(
   A=Saturate(None=Add(A=@n10, B=@n11))
   B=DynamicParameter() {defaultValue=<rgba 2 0 1 1>, paramNames [UScale,Dissolve,Param3,Param4]}.Dissolve
  )
  n13 = TextureCoordinate()
  n14 = ComponentMask(None=@n13) {g=true}
  n15 = OneMinus(None=Sine(None=ComponentMask(None=@n13) {r=true}) {period=2})
  n16 = Multiply(A=Subtract(A=Sine(None=@n14) {period=2}, B=LinearInterpolate(Alpha=@n15) {constA=-4}) {constB=0}) {constB=5}
  DepthFade(Opacity=Multiply(A=Multiply(A=@n1.A, B=@n12), B=Saturate(None=@n16)), FadeDistance=ScalarParameter() {parameterName=FadeDistance}) {fadeDistanceDefault=0}
  n17 = VectorParameter() {parameterName=Color Edge, defaultValue=<rgba 0.1768 0.0024 0 1>}
  n18 = VectorParameter() {parameterName=Color Core, defaultValue=<rgba 0.4271 0.0055 0 1>}
  n19 = Saturate(None=@n10)
  MaterialFunctionCall(Emissive=Multiply(A=LinearInterpolate(A=@n17.RGB, B=@n18.RGB, Alpha=@n19), B=@n1.RGB)) {materialFunction=&MF_FX_EmissiveAdjustments}
 ]
}
