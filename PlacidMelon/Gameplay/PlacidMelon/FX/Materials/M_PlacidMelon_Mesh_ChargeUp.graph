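# Unlit translucent, two-sided Niagara mesh-particle material for PlacidMelon's charge-up effect, drawing soft, particle-colored streaks.
# A UV mask from sine across V and power along U is shaped by dynamic Erode Width and Length.
# A panning smoke texture, randomized per particle in offset and speed, breaks up that mask by Noise Amount.
# Depth fade softens intersections, eye-adaptation inverse compensates brightness, and WPO Bend scales an instance-space offset along U.
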Material /PlacidMelon/Gameplay/PlacidMelon/FX/Materials/M_PlacidMelon_Mesh_ChargeUp
graph {
 properties {
  bAutomaticallySetUsageInEditor=false
  bCastRayTracedShadows=false
  bUsedWithNiagaraMeshParticles=true
  blendMode=BLEND_Translucent
  shadingModel=MSM_Unlit
  twoSided=true
  usedShadingModels=Unlit
 }
 nodes [
  n1 = ParticleColor()
  n2 = TextureCoordinate()
  n3 = Sine(None=ComponentMask(None=@n2) {g=true})
  n4 = DynamicParameter() {paramNames [Erode Width, Erode Length, Noise Amount, WPO Bend]}
  n5 = ComponentMask(None=@n2) {r=true}
  n6 = Power(Base=@n5, Exp=@n4.Erode Length)
  n7 = Multiply(A=@n1.A, B=Multiply(A=Power(Base=@n3, Exp=@n4.Erode Width), B=Saturate(None=@n6)))
  n8 = ParticleRandom()
  n9 = Multiply(
   A=TextureCoordinate() {uTiling=0.5}
   B=AppendVector(A=ScalarParameter() {defaultValue=0.5, parameterName=Tile X}, B=ScalarParameter() {defaultValue=1, parameterName=Tile Y})
  )
  n10 = AppendVector(A=Add(A=Multiply(A=@n8) {constB=67}, B=ComponentMask(None=@n9) {r=true}), B=ComponentMask(None=@n9) {g=true})
  n11 = ParticleRandom()
  n12 = AppendVector(
   A=LinearInterpolate(
    A=ScalarParameter() {defaultValue=-0.2, parameterName=Pan Speed X Min}
    B=ScalarParameter() {defaultValue=-0.8, parameterName=Pan Speed X Max}
    Alpha=Saturate(None=@n11)
   ) {constA=-0.2, constB=-0.8}
   B=ScalarParameter() {parameterName=Pan Speed Y}
  )
  n13 = Multiply(A=TextureSample(UVs=Panner(Coordinate=@n10, Speed=@n12)) {texture=/Game/Effects/Fort_Effects/Textures/Smoke/T_Smoke_Trail_Soft_2}.G) {constB=2}
  n14 = LinearInterpolate(A=@n7, B=Multiply(A=@n7, B=Power(Base=@n13)), Alpha=@n4.Noise Amount)
  DepthFade(Opacity=@n14, FadeDistance=ScalarParameter() {defaultValue=50, parameterName=DepthFade})
  n15 = Saturate(None=@n14)
  EyeAdaptationInverse(LightValueInput=Multiply(A=@n1.RGB, B=Add(A=@n15) {constB=0.1}))
  OneMinus(None=@n15)
  n16 = OneMinus(None=@n5)
  n17 = Constant3Vector() {constant=<rgba 0 -10 0 0>}
  Transform(None=Multiply(A=Multiply(A=@n16, B=@n17), B=@n4.WPO Bend)) {transformSourceType=TRANSFORMSOURCE_Instance}
 ]
}
