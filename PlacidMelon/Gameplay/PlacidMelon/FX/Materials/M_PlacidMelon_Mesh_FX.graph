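# M_PlacidMelon_Mesh_FX is an unlit, translucent, two-sided Niagara material for PlacidMelon mesh effects, also usable on sprites and ribbons.
# Panning cloud noise, shaped by a vertical UV gradient and power, is faded by inverted Fresnel and particle color to drive an exposure-compensated three-color ramp.
# Opacity combines noise, a second inverted Fresnel, particle alpha and depth fade, with a sided-sign switch.
# Dynamic parameters adjust noise and gradient per particle.
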Material /PlacidMelon/Gameplay/PlacidMelon/FX/Materials/M_PlacidMelon_Mesh_FX
graph {
 properties {
  bCastRayTracedShadows=false
  bUsedWithNiagaraMeshParticles=true
  bUsedWithNiagaraRibbons=true
  bUsedWithNiagaraSprites=true
  blendMode=BLEND_Translucent
  pixelDepthOffsetMode=PDOM_Legacy
  shadingModel=MSM_Unlit
  twoSided=true
  usedShadingModels=Unlit
 }
 nodes [
  n1 = TextureCoordinate()
  n2 = VertexNormalWS()
  n3 = DynamicParameter() {defaultValue=<rgba 1 0 0 0>, paramNames [Noise Amount, "Gradient Mask Pwr (Add)", Opacity Fresnel Power, Noise Power]}
  Add(A=@n1, B=Multiply(A=ComponentMask(None=@n2) {g=true, r=true}, B=@n3."Gradient Mask Pwr (Add)"))
  n4 = Fresnel(ExponentIn=ScalarParameter() {defaultValue=0.25, parameterName=Fresnel Exponent}) {exponent=0.25}
  n5 = AppendVector(A=ScalarParameter() {defaultValue=3, parameterName=Noise Tile X}, B=ScalarParameter() {defaultValue=0.1, parameterName=Noise Tile Y})
  n6 = ScalarParameter() {defaultValue=0.03, parameterName=Noise Speed X}
  n7 = ScalarParameter() {defaultValue=0.15, parameterName=Noise Speed Y}
  MaterialExpressionTextureSampleParameter2D_0 = TextureSampleParameter2D(UVs=Panner(Coordinate=Multiply(A=@n5, B=@n1), Speed=AppendVector(A=@n6, B=@n7))) {
   parameterName=NoiseTexture
   samplerType=SAMPLERTYPE_LinearGrayscale
   texture=/Game/Effects/Fort_Effects/Emotes/UltraLight/Texture/T_Cloud_Greyscale
  }
  n8 = TextureCoordinate()
  n9 = ScalarParameter() {defaultValue=1, parameterName=GradientMaskPower}
  n10 = DynamicParameter() {defaultValue=<rgba 1 0 0 0>, paramNames [Noise Amount, "Gradient Mask Pwr (Add)", Opacity Fresnel Power, Noise Power]}
  n11 = Saturate(None=Power(Base=ComponentMask(None=@n8) {g=true}, Exp=Add(A=@n9, B=@n10."Gradient Mask Pwr (Add)")))
  n12 = Switch!(
   SwitchValue=ScalarParameter() {parameterName=Noise Blend}
   Default=Multiply(A=@MaterialExpressionTextureSampleParameter2D_0.G, B=@n11)
   None=@MaterialExpressionTextureSampleParameter2D_0.G
  ) {inputs [{
     inputName=None
     input {
      expression=&self:MaterialExpressionTextureSampleParameter2D_0
      inputName=None
      mask=1
      maskA=0
      maskB=0
      maskG=1
      maskR=0
      outputIndex=2
     }
    }]}
  n13 = DynamicParameter() {defaultValue=<rgba 1 0 0 0>, paramNames [Noise Amount, "Gradient Mask Pwr (Add)", Opacity Fresnel Power, Noise Power]}
  n14 = Add(A=@n13.Noise Power, B=ScalarParameter() {defaultValue=1, parameterName=Noise Power})
  n15 = Multiply(
   A=Multiply(A=Power(Base=@n12, Exp=@n14) {constExponent=1}, B=ScalarParameter() {defaultValue=5, parameterName=NoiseMultiply}) {constB=5}
   B=@n11
  )
  n16 = Multiply(A=OneMinus(None=Saturate(None=@n4)), B=@n15)
  n17 = ParticleColor()
  EyeAdaptationInverse(LightValueInput=MaterialFunctionCall(
    A=VectorParameter() {parameterName=Color C, defaultValue=<rgba 0.2752 0 0 1>}.RGB
    B=VectorParameter() {parameterName=Color B, defaultValue=<rgba 1 0 0 1>}.RGB
    C=VectorParameter() {parameterName=Color A, defaultValue=<rgba 1 1 1 1>}.RGB
    Alpha=Multiply(A=@n16, B=@n17.RGB)
   ) {materialFunction=&Lerp_3Color}.Result)
  n18 = Power(
   Base=Fresnel(ExponentIn=ScalarParameter() {defaultValue=0.25, parameterName=Opacity Fresnel Exponent}) {exponent=0.25}
   Exp=ScalarParameter() {defaultValue=1, parameterName=Opacity Fresnel Power}
  )
  n19 = Add(A=OneMinus(None=Saturate(None=@n18)), B=@n13.Noise Amount)
  n20 = DepthFade(Opacity=Multiply(A=Multiply(A=@n19, B=@n17.A), B=@n15), FadeDistance=ScalarParameter() {defaultValue=100, parameterName=Depth Fade}) {fadeDistanceDefault=25}
  MaterialExpressionMultiply_3 = Multiply(A=@n20, B=TwoSidedSign())
  Switch(SwitchValue=ScalarParameter() {defaultValue=1, parameterName=Double Sided}, Default=@n20, None=@MaterialExpressionMultiply_3) {inputs [{
     inputName=None
     input {
      expression=&self:MaterialExpressionMultiply_3
      inputName=None
      mask=0
      maskA=0
      maskB=0
      maskG=0
      maskR=0
      outputIndex=0
     }
    }]}
 ]
}
