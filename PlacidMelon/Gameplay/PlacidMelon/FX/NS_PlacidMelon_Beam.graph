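# Niagara beam effect for PlacidMelon, drawing ribbons from a sampled skeletal mesh position to the User.BeamEnd_0 position.
# Beams bursts two 20-particle ribbons placed along a Hermite spline, using spawn-group masks to give the projectile and fade layers separate width, color, alpha and material animation across two renderers.
# PP attaches a component renderer to the skeletal mesh, spawning only when User.LocalView is true.
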NiagaraSystem /PlacidMelon/Gameplay/PlacidMelon/FX/NS_PlacidMelon_Beam
dependencies {
 usedDataInterfaces [&NiagaraDataInterfaceCurve, &NiagaraDataInterfaceSkeletalMesh]
 usedDynamicInputs [
  &VectorLength
  &Subtract_Vector
  &FindQuatBetweenTwoVectors
  /Niagara/DynamicInputs/Lerp/Lerp_Position
  &MakeFloatFromInt
 ]
 usedModules [
  &self:SystemSpawnScript.SystemScriptSource.SystemScriptGraph.NiagaraNodeAssignment_3.SetVariables_#1
  &SystemState
  &SampleSkeletalMesh
  &self:SystemSpawnScript.SystemScriptSource.SystemScriptGraph.NiagaraNodeAssignment_0.SetVariables_#2
  &self:SystemSpawnScript.SystemScriptSource.SystemScriptGraph.NiagaraNodeAssignment_1.SetVariables_#3
  &Timeline
 ]
}
emitters {
 Beams {
  emitterUpdateScript [
   &EmitterState()
   &SpawnBurst_Instantaneous(Spawn Count=20, Use Spawn Probability=false, UseLoopCountLimit=false)
   &SpawnBurst_Instantaneous(Spawn Count=20, Spawn Group=2, Use Spawn Probability=false, UseLoopCountLimit=false) {module=SpawnBurst_Instantaneous002}
  ]
  particleSpawnScript [
   /SetVariables(Particles.RibbonID=/Niagara/DynamicInputs/Meshes/MakeRibbonIDFromSpawnGroup()) {isSetParameters=true}
   &PartitionParticles(
    Number Of Partitions=1
    Particles Per Partition=20
    Partition Mode:ENiagara_PartitionMode=::NewEnumerator1
    Write Partition Index to Ribbon Link Order:ENiagara_AttributeSamplingApplyOutput=::NewEnumerator0
   )
   &InitializeParticle(
    Color=&MaskLinearColorBySpawnGroup(
     Masked Linear Color=<rgba 1 1 1 0.3>
     Passthrough Linear Color=&MaskLinearColorBySpawnGroup(Masked Linear Color=<rgba 0 0.2333 0.7 0.8>, Passthrough Linear Color=<rgba 1 1 1 0.15>, Spawn Group Mask=1)
    )
    Lifetime=&MaskFloatBySpawnGroup(Masked Float=0.6, Passthrough Float=0.9)
    Ribbon Width=&MaskFloatBySpawnGroup(Masked Float=140, Passthrough Float=90, Spawn Group Mask=1)
    Ribbon Width Mode=::NewEnumerator1
    UsePositionOffset=false
    Write Material Random=true
   )
  ]
  particleUpdateScript [
   &ParticleState()
   /Game/Effects/Niagara/Modules/Curves/FN_HermiteSpline(
    Position Blend:NiagaraFloat=$Particles.Initial.PartitionParticles.Normalized Index In Partition
    Position End:NiagaraPosition=$User.BeamEnd_0
    Position Start:Vector3f=&Add_Vector(
     A=Linked() {linkedVariable {name=System.SampleSkeletalMesh.SampledPosition, type=NiagaraPosition}}
     B=&RotateVectorByQuaternion(Vector To Rotate=<0 0 0>, Quaternion=$System.BeamOrientation)
    )
    Write Particles Position=true
   )
   /SetVariables(
    Particles.Beam Fade_Width:NiagaraFloat=&Multiply_Float(
     A=$Particles.Initial.RibbonWidth
     B=&MaskFloatBySpawnGroup(
      Masked Float=&FloatFromCurve(FloatCurve=DataInterface() {propertyValues {Curve {keys [<at 0.4 0.7 RCIM_Cubic>, {arriveTangent=0.0966, interpMode=RCIM_Cubic, leaveTangent=0.0966, tangentMode=RCTM_User, time=1, value=0}]}}})
      Passthrough Float=0
      Spawn Group Mask=2
     )
    )
    Particles.ProjectileBeam_Width:NiagaraFloat=&Multiply_Float(
     A=$Particles.Initial.RibbonWidth
     B=&MaskFloatBySpawnGroup(
      Masked Float=&FloatFromCurve(FloatCurve=DataInterface() {propertyValues {Curve {keys [{arriveTangent=-10.5006, interpMode=RCIM_Cubic, leaveTangent=-10.5006, tangentMode=RCTM_User, time=0, value=1}, <at 0.2 0.3>]}}})
      Passthrough Float=0
     )
    )
   ) {isSetParameters=true}
   &ScaleColor(
    Scale Alpha=&MaskFloatBySpawnGroup(
     Masked Float=&Multiply_Float(
      A=&FloatFromCurve(
       FloatCurve=DataInterface() {propertyValues {Curve {keys [
           {arriveTangent=0.1111, interpMode=RCIM_Cubic, leaveTangent=0.1111, leaveTangentWeight=0.7387, tangentMode=RCTM_User, time=0, value=1}
           {interpMode=RCIM_Cubic, tangentMode=RCTM_User, time=0.2, value=1}
           <at 0.8 1 RCIM_Cubic>
           <at 1 1>
          ]}}}
       CurveIndex=&Multiply_Float(A=$Particles.PartitionParticles.Normalized Index In Partition, B=1)
       Scale Curve=10
      )
      B=&FloatFromCurve(
       FloatCurve=DataInterface() {propertyValues {Curve {keys [
           {arriveTangent=0.1111, interpMode=RCIM_Cubic, leaveTangent=0.1111, leaveTangentWeight=0.7387, tangentMode=RCTM_User, time=0, value=1}
           <at 0.2 1>
           {arriveTangentWeight=0.2134, interpMode=RCIM_Cubic, leaveTangentWeight=0.2134, tangentMode=RCTM_User, time=0.6, value=1}
           <at 1 0>
          ]}}}
       CurveIndex=$Particles.PartitionParticles.Normalized Index In Partition
       Scale Curve=1.2
      )
     )
     Passthrough Float=&MaskFloatBySpawnGroup(
      Masked Float=&FloatFromCurve(
       FloatCurve=DataInterface() {propertyValues {Curve {keys [
           {arriveTangent=0.1111, interpMode=RCIM_Cubic, leaveTangent=0.1111, leaveTangentWeight=0.7387, tangentMode=RCTM_User, time=0, value=1}
           <at 0.2 1>
           {arriveTangentWeight=0.2134, interpMode=RCIM_Cubic, leaveTangentWeight=0.2134, tangentMode=RCTM_User, time=0.6, value=1}
           <at 1 0>
          ]}}}
       CurveIndex=$Particles.PartitionParticles.Normalized Index In Partition
      )
      Passthrough Float=&Multiply_Float(
       A=&Multiply_Float(
        A=10
        B=&FloatFromCurve(FloatCurve=DataInterface() {propertyValues {Curve {keys [
             {arriveTangent=-1.5385, interpMode=RCIM_Cubic, leaveTangent=-1.5385, tangentMode=RCTM_User, time=0, value=1}
             {arriveTangent=-2.602, interpMode=RCIM_Cubic, leaveTangent=-2.602, tangentMode=RCTM_User, time=1, value=0}
            ]}}})
       )
       B=&FloatFromCurve(
        FloatCurve=DataInterface() {propertyValues {Curve {keys [
            {arriveTangent=0.1111, interpMode=RCIM_Cubic, leaveTangent=0.1111, leaveTangentWeight=0.7387, tangentMode=RCTM_User, time=0, value=1}
            <at 0.2 1>
            {arriveTangentWeight=0.2134, interpMode=RCIM_Cubic, leaveTangentWeight=0.2134, tangentMode=RCTM_User, time=0.6, value=1}
            <at 1 0>
           ]}}}
        CurveIndex=$Particles.PartitionParticles.Normalized Index In Partition
        Scale Curve=1.8
       )
      )
      Spawn Group Mask=1
     )
    )
    Scale RGB=&VectorFromFloat(Value=&MaskFloatBySpawnGroup(
      Masked Float=&FloatFromCurve(FloatCurve=DataInterface() {propertyValues {Curve {keys [<at 0.1 5>,<at 0.12 2>]}}})
      Passthrough Float=1
      Spawn Group Mask=2
     ))
    ScaleA=true
    ScaleRGB=true
   )
   &DynamicMaterialParameters(
    Index 0 Param 1=&MaskFloatBySpawnGroup(Masked Float=2, Passthrough Float=1)
    Index 0 Param 2=&MaskFloatBySpawnGroup(
     Masked Float=&FloatFromCurve(FloatCurve=DataInterface() {propertyValues {Curve {keys [
          <at 0.3 0 RCIM_Cubic>
          {arriveTangent=1.4286, interpMode=RCIM_Cubic, leaveTangent=1.4286, time=0.8968, value=0.3113}
          {arriveTangent=4.0857, tangentMode=RCTM_User, time=1, value=1}
         ]}}})
     Passthrough Float=&MaskFloatBySpawnGroup(
      Masked Float=&FloatFromCurve(FloatCurve=DataInterface() {propertyValues {Curve {keys [<at 0.3 0 RCIM_Cubic>, {arriveTangent=2.2097, tangentMode=RCTM_User, time=1, value=0.9}]}}})
      Passthrough Float=&FloatFromCurve(FloatCurve=DataInterface() {propertyValues {Curve {keys [<at 0.3 0 RCIM_Cubic>, {arriveTangent=2.2097, tangentMode=RCTM_User, time=1, value=0.8}]}}})
      Spawn Group Mask=1
     )
    )
    Module.Index0Param0Write=true
    Module.Index0Param1Write=true
    Module.Index0Param2Write=false
    Module.Index0Param3Write=false
   )
  ]
  rendererClasses [&NiagaraRibbonRendererProperties]
  renderers [
   {rendererClass=&NiagaraRibbonRendererProperties}
   {rendererClass=&NiagaraRibbonRendererProperties, rendererIndex=1}
  ]
 }
 PP {
  emitterUpdateScript [
   &EmitterState(Life Cycle Mode=::NewEnumerator1, Loop Behavior=::NewEnumerator1, Loop Duration=1, UseLoopDelay=false)
   &SpawnBurst_Instantaneous(
    Spawn Probability=&MakeCustomFloatFromBool(Bool=$User.LocalView, True Float=1, False Float=0)
    Use Spawn Probability=true
    UseLoopCountLimit=false
   )
  ]
  particleSpawnScript [&InitializeParticle(
    Color Mode=::NewEnumerator0
    Lifetime=0.5
    Sprite Size Mode=::NewEnumerator3
    Uniform Sprite Size=200
    UsePositionOffset=false
    Write Material Random=true
   )]
  particleUpdateScript [
   /SetVariables(Particles.SpeedNormalizedB:NiagaraFloat=&FloatFromCurve(FloatCurve=DataInterface() {propertyValues {Curve {keys [
         {arriveTangent=-1.7682, interpMode=RCIM_Cubic, leaveTangent=-1.7682, tangentMode=RCTM_User, time=0.2, value=0.5}
         {interpMode=RCIM_Cubic, tangentMode=RCTM_User, time=1, value=0}
        ]}}})) {isSetParameters=true}
   &SkeletalMeshLocation(MaskSpawnGroup=false, Mesh Sampling Type=::NewEnumerator5, Skeletal Mesh=$System.Skeletal Mesh, Socket Sampling Mode=::NewEnumerator5)
   &ParticleState()
   /SetVariables(Particles.ComponentsEnabled:NiagaraBool=&SetBoolByFloatComparison(
     A=$Particles.Age
     B=&Subtract_Float(A=$Particles.Lifetime, B=$Engine.DeltaTime)
     Comparison Type=::NewEnumerator6
    )) {isSetParameters=true}
  ]
  rendererClasses [&NiagaraComponentRendererProperties]
  renderers [{rendererClass=&NiagaraComponentRendererProperties, rendererIndex=0}]
 }
}
summary {
 systemName=NS_PlacidMelon_Beam
 emitters [
  {emitterName=PP, rendererClasses [&NiagaraComponentRendererProperties]}
  {emitterName=Beams, rendererClasses [&NiagaraRibbonRendererProperties]}
 ]
 userVariables [
  User.BeamEnd_0 {type=NiagaraPosition, description="", defaultValue {}}
  User.BeamFadeOut {type=NiagaraInt32}
  User.LocalView {type=NiagaraBool}
 ]
}
