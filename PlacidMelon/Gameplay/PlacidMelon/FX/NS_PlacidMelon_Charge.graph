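# Niagara system for a PlacidMelon charge effect, built from eight emitters.
# The emitter names indicate red and white spheres, glow, light, sparks, two mesh fin layers and spawn streaks, though no renderers are recorded for them.
# It depends on a skeletal mesh data interface, alongside SystemState and a variable assignment in the system spawn script.
# It lists no user variables or dynamic inputs.
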NiagaraSystem /PlacidMelon/Gameplay/PlacidMelon/FX/NS_PlacidMelon_Charge
dependencies {
 usedDataInterfaces [&NiagaraDataInterfaceSkeletalMesh]
 usedDynamicInputs []
 usedModules [
  &self:SystemSpawnScript.SystemScriptSource.SystemScriptGraph.NiagaraNodeAssignment_0.SetVariables_#1
  &SystemState
 ]
}
emitters {
 Glow {rendererClasses [], renderers []}
 Light {rendererClasses [], renderers []}
 Mesh_Fins {rendererClasses [], renderers []}
 Mesh_Fins4 {rendererClasses [], renderers []}
 Sparks {rendererClasses [], renderers []}
 Spawn_Streaks {rendererClasses [], renderers []}
 Sphere_Red {rendererClasses [], renderers []}
 Sphere_White001 {rendererClasses [], renderers []}
}
summary {
 systemName=NS_PlacidMelon_Charge
 emitters [
  {emitterName=Sphere_Red, rendererClasses []}
  {emitterName=Sphere_White001, rendererClasses []}
  {emitterName=Light, rendererClasses []}
  {emitterName=Glow, rendererClasses []}
  {emitterName=Sparks, rendererClasses []}
  {emitterName=Mesh_Fins4, rendererClasses []}
  {emitterName=Mesh_Fins, rendererClasses []}
  {emitterName=Spawn_Streaks, rendererClasses []}
 ]
 userVariables []
}
