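# Niagara system for the impact effect of the PlacidMelon gameplay feature.
# It layers thirteen emitters whose names indicate glow, decal, light, red cones and jet, top and bottom smoke rings, smoke, sparks, refraction, facing shockwave and mesh fins.
# SystemState is the only module listed as a dependency.
# No user variables are exposed, and no renderers or data interfaces are listed.
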NiagaraSystem /PlacidMelon/Gameplay/PlacidMelon/FX/NS_PlacidMelon_Impact
dependencies {usedDataInterfaces [], usedDynamicInputs [], usedModules [&SystemState]}
emitters {
 Decal {rendererClasses [], renderers []}
 Glow {rendererClasses [], renderers []}
 Light001 {rendererClasses [], renderers []}
 Mesh_Fins003 {rendererClasses [], renderers []}
 RedCone_Base {rendererClasses [], renderers []}
 RedCone_Start {rendererClasses [], renderers []}
 RedJet_Middle {rendererClasses [], renderers []}
 Refraction {rendererClasses [], renderers []}
 Shockwave_Facing {rendererClasses [], renderers []}
 Smoke001 {rendererClasses [], renderers []}
 SmokeRing_Bottom {rendererClasses [], renderers []}
 SmokeRing_Top {rendererClasses [], renderers []}
 Sparks {rendererClasses [], renderers []}
}
summary {
 systemName=NS_PlacidMelon_Impact
 emitters [
  {emitterName=Glow, rendererClasses []}
  {emitterName=Decal, rendererClasses []}
  {emitterName=RedCone_Base, rendererClasses []}
  {emitterName=RedJet_Middle, rendererClasses []}
  {emitterName=SmokeRing_Bottom, rendererClasses []}
  {emitterName=SmokeRing_Top, rendererClasses []}
  {emitterName=RedCone_Start, rendererClasses []}
  {emitterName=Refraction, rendererClasses []}
  {emitterName=Shockwave_Facing, rendererClasses []}
  {emitterName=Sparks, rendererClasses []}
  {emitterName=Smoke001, rendererClasses []}
  {emitterName=Mesh_Fins003, rendererClasses []}
  {emitterName=Light001, rendererClasses []}
 ]
 userVariables []
}
