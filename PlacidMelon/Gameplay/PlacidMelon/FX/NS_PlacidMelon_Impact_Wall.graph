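# Niagara system for the PlacidMelon wall impact effect, judging by its name and path.
# It layers twelve emitters whose names indicate a glow, light, blast ring, white mesh cone, four fin mesh variants, sparks, embers, smoky glow and decal.
# It declares no user variables, and its only listed module is SystemState.
# No renderers or data interfaces are recorded, so each layer's role rests on its emitter name alone.
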NiagaraSystem /PlacidMelon/Gameplay/PlacidMelon/FX/NS_PlacidMelon_Impact_Wall
dependencies {usedDataInterfaces [], usedDynamicInputs [], usedModules [&SystemState]}
emitters {
 BlastRing {rendererClasses [], renderers []}
 Decal {rendererClasses [], renderers []}
 Embers {rendererClasses [], renderers []}
 Glow {rendererClasses [], renderers []}
 Light {rendererClasses [], renderers []}
 MeshCone_White {rendererClasses [], renderers []}
 Mesh_Fins {rendererClasses [], renderers []}
 Mesh_Fins001 {rendererClasses [], renderers []}
 Mesh_Fins002 {rendererClasses [], renderers []}
 Mesh_Fins003 {rendererClasses [], renderers []}
 SmokeyGlow {rendererClasses [], renderers []}
 Sparks {rendererClasses [], renderers []}
}
summary {
 systemName=NS_PlacidMelon_Impact_Wall
 emitters [
  {emitterName=Glow, rendererClasses []}
  {emitterName=Mesh_Fins003, rendererClasses []}
  {emitterName=MeshCone_White, rendererClasses []}
  {emitterName=BlastRing, rendererClasses []}
  {emitterName=Sparks, rendererClasses []}
  {emitterName=SmokeyGlow, rendererClasses []}
  {emitterName=Decal, rendererClasses []}
  {emitterName=Mesh_Fins001, rendererClasses []}
  {emitterName=Embers, rendererClasses []}
  {emitterName=Mesh_Fins, rendererClasses []}
  {emitterName=Light, rendererClasses []}
  {emitterName=Mesh_Fins002, rendererClasses []}
 ]
 userVariables []
}
