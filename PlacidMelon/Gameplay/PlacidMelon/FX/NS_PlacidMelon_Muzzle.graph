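# Niagara system for the PlacidMelon muzzle effect, by its name a muzzle flash.
# It is built from three emitters named Muzzle, Light and Glow, with no renderer classes listed for any of them.
# The only listed module is SystemState, and the system exposes no user variables and uses no data interfaces or dynamic inputs.
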NiagaraSystem /PlacidMelon/Gameplay/PlacidMelon/FX/NS_PlacidMelon_Muzzle
dependencies {usedDataInterfaces [], usedDynamicInputs [], usedModules [&SystemState]}
emitters {Glow {rendererClasses [], renderers []}, Light {rendererClasses [], renderers []}, Muzzle {rendererClasses [], renderers []}}
summary {
 systemName=NS_PlacidMelon_Muzzle
 emitters [{emitterName=Muzzle, rendererClasses []}, {emitterName=Light, rendererClasses []}, {emitterName=Glow, rendererClasses []}]
 userVariables []
}
